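# Niagara system for the idle effect on the Street Racer Drift pickaxe cosmetic, named for electric arcs and flares.
# Three emitters, apparently lightweight by their LWE suffix, cover a sprite flare, sprite electric arcs and decal electric arcs, with no renderers listed.
# The system uses SystemState and a set-variables assignment, plus a random-range float input and a socket reader data interface.
# It exposes no user parameters.
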NiagaraSystem /BRCosmetics/Weapons/FORT_Melee/Pickaxe_Street_Racer_Drift/FX/NS_Pickaxe_Drift_Idle
dependencies {
 usedDataInterfaces [&NiagaraDataInterfaceSocketReader]
 usedDynamicInputs [&RandomRangeFloat]
 usedModules [
  &self:SystemSpawnScript.SystemScriptSource.SystemScriptGraph.NiagaraNodeAssignment_0.SetVariables_#1
  &SystemState
 ]
}
emitters {
 Decal_ElectricArc_LWE {rendererClasses [], renderers []}
 Sprite_ElectricArc_LWE {rendererClasses [], renderers []}
 Sprite_Flare_LWE {rendererClasses [], renderers []}
}
summary {
 systemName=NS_Pickaxe_Drift_Idle
 emitters [
  {emitterName=Sprite_Flare_LWE, rendererClasses []}
  {emitterName=Sprite_ElectricArc_LWE, rendererClasses []}
  {emitterName=Decal_ElectricArc_LWE, rendererClasses []}
 ]
 userVariables []
}
