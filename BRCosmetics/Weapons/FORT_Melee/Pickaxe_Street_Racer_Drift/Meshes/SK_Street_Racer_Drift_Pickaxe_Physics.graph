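# Physics asset for the Street Racer Drift two-handed pickaxe, giving its hanging chain bodies and joints for physical motion.
# One box covers the handle bone, and six similarly sized capsules cover the bones chain02 through chain07.
# Constraints link the capsules in series from the handle out to chain07.
# Every joint locks its first swing axis, so each link cannot bend in that direction relative to its parent.
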PhysicsAsset /BRCosmetics/Weapons/FORT_Melee/Pickaxe_Street_Racer_Drift/Meshes/SK_Street_Racer_Drift_Pickaxe_Physics
bodies {
 Melee_TwoHanded_Handle [{shapeType=Box, extentX=8.2966, extentY=81.1438, extentZ=135.4456, center=<-0.1461 14.0252 22.3914>}]
 chain02 [{shapeType=Capsule, length=4.0579, radius=2.6683, center=<-0.3702 -0.0025 -3.3143>, rotation=<rot -0.9111 -90 1.436>}]
 chain03 [{shapeType=Capsule, length=4.0539, radius=2.6837, center=<-0.1396 -0.0023 -3.2791>, rotation=<rot 0.2004 -90 1.195>}]
 chain04 [{shapeType=Capsule, length=4.048, radius=2.6719, center=<-0.452 -0.0024 -3.2929>, rotation=<rot 0.9767 -90 -0.2438>}]
 chain05 [{shapeType=Capsule, length=4.0356, radius=2.65, center=<-0.1396 -0.0023 -3.1812>, rotation=<rot -0.1995 -90 -0.5751>}]
 chain06 [{shapeType=Capsule, length=4.0672, radius=2.6836, center=<-0.3637 -0.0022 -3.2853>, rotation=<rot -1.1936 -90 -1.554>}]
 chain07 [{shapeType=Capsule, length=4.4234, radius=2.7312, center=<-0.1396 -0.0024 -3.6713>, rotation=<rot 0 -90 1.8654>}]
}
constraints [
 {bone1Name=chain07, bone2Name=chain06, swing1LimitDegrees=45, swing1Motion=Locked, swing2LimitDegrees=45, twistLimitDegrees=45}
 {bone1Name=chain02, bone2Name=Melee_TwoHanded_Handle, swing1LimitDegrees=45, swing1Motion=Locked, swing2LimitDegrees=45, twistLimitDegrees=45}
 {bone1Name=chain03, bone2Name=chain02, swing1LimitDegrees=45, swing1Motion=Locked, swing2LimitDegrees=45, twistLimitDegrees=45}
 {bone1Name=chain04, bone2Name=chain03, swing1LimitDegrees=45, swing1Motion=Locked, swing2LimitDegrees=45, twistLimitDegrees=45}
 {bone1Name=chain05, bone2Name=chain04, swing1LimitDegrees=45, swing1Motion=Locked, swing2LimitDegrees=45, twistLimitDegrees=45}
 {bone1Name=chain06, bone2Name=chain05, swing1LimitDegrees=45, swing1Motion=Locked, swing2LimitDegrees=45, twistLimitDegrees=45}
]
