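# Animation Blueprint for the Street Racer Drift pickaxe that adds secondary motion to its chain bones.
# The input pose feeds either a RigidBody simulation or a Trail controller on a seven-bone chain, preceded by an additive rotation of chain01.
# A bool blend on bRigidBodyEnabled selects between them.
# An event graph branch on view distance quality sets RigidBody_yes or RigidBody_no, which the pose graph does not read.
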AnimBlueprint /BRCosmetics/Weapons/FORT_Melee/Pickaxe_Street_Racer_Drift/Meshes/SK_Street_Racer_Drift_Pickaxe_Skeleton_ABP
target_skeleton=&/SK_Street_Racer_Drift_Pickaxe_Skeleton
graphs [
 AnimGraph {nodes [
   n1 = RigidBody(ComponentPose=Local To Component(LocalPose=Use cached pose 'inputPose'())) {
    overridePhysicsAsset=&/SK_Street_Racer_Drift_Pickaxe_Physics
    alphaBoolBlend {blendOption=Linear}
    componentAppliedLinearAccClamp=<1600 1600 1600>
    componentLinearAccScale=<1 1 1>
    componentLinearVelScale=<1 1 1>
   }
   n2 = "Transform (Modify) Bone - Bone: chain01"(ComponentPose=Local To Component(LocalPose=Use cached pose 'inputPose'()), Alpha=1, Translation="0,0,0", Rotation="-66,33,0", Scale="1,1,1") {rotationMode=BMM_Additive, alphaBoolBlend {blendOption=Linear}, boneToModify {boneName=chain01}, rotation {pitch=-66, yaw=33}}
   n3 = Trail controller - Bone: chain07(ComponentPose=@n2, Alpha=1) {
    bInvertChainBoneAxis=true
    chainBoneAxis=Z
    chainLength=7
    relaxationSpeedScale=4
    rotationLimits [
     {limitMax=<180 180 180>}
     {limitMax=<180 180 180>}
     {limitMax=<180 180 180>}
     {limitMax=<180 180 180>}
     {limitMax=<180 180 180>}
     {limitMax=<180 180 180>}
     {limitMax=<180 180 180>}
    ]
    rotationOffsets [<0 0 0>, <0 0 0>, <0 0 0>, <0 0 0>, <0 0 0>, <0 0 0>, <0 0 0>]
    trailBone {boneName=chain07}
    trailRelaxationSpeed {editorCurveData {keys [{arriveTangent=-11.8836, interpMode=RCIM_Cubic, leaveTangent=-11.8836, tangentMode=RCTM_User, time=0, value=10}, <at 1 5 RCIM_Linear>]}}
   }
   Output Pose(Result=Blend Poses by bool(BlendPose_0=Component To Local(ComponentPose=@n1), BlendPose_1=Component To Local(ComponentPose=@n3), bActiveValue=Get bRigidBodyEnabled()) {blendType=Linear}) {layerGroup=None}
   AnimGraphNode_SubInput_0 = Input Pose()
   Save cached pose 'inputPose'(Pose=@AnimGraphNode_SubInput_0) {cacheName=inputPose}
  ]}
 EventGraph {nodes [
   Event BlueprintInitializeAnimation()
   n1 = GetFortGameUserSettings!() -> (ReturnValue:Fort Game User Settings Object Reference)
   > n2 = Branch(Condition="integer >= integer"!(A:int=GetViewDistanceQuality!(self:Game User Settings Object Reference=@n1) -> (ReturnValue:int=0), B:int=2) -> (ReturnValue:bool) {is=title K2Node_CallFunction})
   > Set RigidBody_yes(RigidBody_yes:double=1) -> (Output_Get:double)
   Set RigidBody_no(execute=@n2.else, RigidBody_no:double=1) -> (Output_Get:double)
   TryGetPawnOwner()
   Event BlueprintUpdateAnimation()
  ]}
]
variables [
 RigidBody_no {type=number}
 RigidBody_yes {type=number}
]
